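# Looping ambient sound cue for the Cattus/Doggus event, playing a Zero Point sphere-center loop at boosted volume.
# A single wave player feeds the output, routed to an extra-high-priority event sound class.
# Overridden attenuation holds full volume nearby, then fades along a custom curve and adds distance-based low-pass filtering.
# Interpolated listener focus makes the loop much quieter off-focus and shortens its perceived distance when faced, with a wide stereo spread.
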SoundCue /Figment_S10/Athena/Sounds/Events/Cattus/CattusDoggus_Ambient_Cue
root=WavePlayer_0
nodes [{
  id=WavePlayer_0
  properties {bLooping=true, soundWaveAssetPtr=/Figment_S10/Athena/Sounds/Events/Cattus/ZeroPoint/Snow_Expl_SphereCenter_Loop_01}
 }]
properties {
 bOverrideAttenuation=true
 maxDistance=6524
 soundClassObject=/Figment_S10/Athena/Sounds/Events/SoundClasses/Fort_Events_Priority_ExtHigh
 volumeMultiplier=2
 attenuationOverrides {
  bAttenuateWithLPF=true
  bEnableFocusInterpolation=true
  bEnableListenerFocus=true
  distanceAlgorithm=Custom
  falloffDistance=5500
  focusAttackInterpSpeed=0.5
  focusAzimuth=15
  focusDistanceScale=0.5
  focusReleaseInterpSpeed=0.25
  lPFFrequencyAtMax=500
  lPFRadiusMax=2000
  lPFRadiusMin=1000
  nonFocusAzimuth=90
  nonFocusVolumeAttenuation=0.1
  stereoSpread=2048
  attenuationShapeExtents {x=1024}
  customAttenuationCurve {editorCurveData {keys [<at 0 1>,<at 0.3 1>,<at 0.4 0.5>,<at 1 0>]}}
 }
}
