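# Sound cue for the fourth Cattus impact in the Cattus vs. Doggus event, playing a single Cattus "Worldus" wave.
# A quality-level root routes its first two slots to the same wave player and leaves the other two empty.
# It uses mapwide event attenuation with a very large max distance, an important in-game mission sound class, and reduced volume.
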SoundCue /Figment_S10/Athena/Sounds/Events/Cattus/CattusDoggus_Cattus_Impact_04_Cue
root=QualityLevel_0
nodes [
 {id=QualityLevel_0, children [WavePlayer_0,WavePlayer_0,null,null]}
 {id=WavePlayer_0, properties {soundWaveAssetPtr=/Figment_S10/Athena/Sounds/Events/Cattus/Cattus_Doggus/Cattus_Doggus_Cattus_Worldus_v1_08}}
]
properties {
 attenuationSettings=/Figment_S10/Athena/Sounds/Events/Attenuations/Events_Mapwide_Att
 maxDistance=2097152
 soundClassObject=&Fort_Mission_Ingame_Important_PSM
 volumeMultiplier=0.4
}
