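# Sound cue for a Cattus stem of the Cattus and Doggus event, played with map-wide event attenuation and an important in-game mission sound class.
# A quality-level switch sends the first two levels straight to the v1_02 wave.
# The last two levels play the differently numbered v1_01_ios wave through an enveloper that ramps volume up from silence over three seconds.
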SoundCue /Figment_S10/Athena/Sounds/Events/Cattus/CattusDoggus_Cattus_Stem_03_Cue
root=QualityLevel_0
nodes [
 {
  id=Enveloper_0
  children [WavePlayer_1]
  properties {pitchCurve {editorCurveData {keys [{time=0}]}}, volumeCurve {editorCurveData {keys [<at 0 0>,<at 3 1.1993>]}}}
 }
 {id=QualityLevel_0, children [WavePlayer_0,WavePlayer_0,Enveloper_0,Enveloper_0]}
 {id=WavePlayer_0, properties {soundWaveAssetPtr=/Figment_S10/Athena/Sounds/Events/Cattus/Cattus_Doggus/Cattus_Doggus_Cattus_v1_02}}
 {id=WavePlayer_1, properties {soundWaveAssetPtr=/Figment_S10/Athena/Sounds/Events/Cattus/Cattus_Doggus/Cattus_Doggus_Cattus_v1_01_ios}}
]
properties {
 attenuationSettings=/Figment_S10/Athena/Sounds/Events/Attenuations/Events_Mapwide_Att
 maxDistance=2097152
 soundClassObject=&Fort_Mission_Ingame_Important_PSM
 volumeMultiplier=0.4
}
